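# Conversation database for the RoyalWonk NPC, defining the greeting and top-level branches shown when a player talks to it.
# The entry point leads to a speech node whose line is picked by prioritized requirements: first-ever conversation, match conversation count, or reactions to specific player outfits.
# That node then links by entry tag to shared quest, basic service and random service conversations, plus RoyalWonk's own dialogue branch.
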ConversationDatabase /NPCLibrary/NPCs/RoyalWonk/NPCLibrary_Conversation_RoyalWonk
entry_points [{tag=Conversation.Entry.RoyalWonk, destinations [#1]}]
nodes [
 ConversationEntryPointNode_7 {
  id=#1
  class=ConversationEntryPointNode
  outputs [#2]
  properties {entryTag {tagName=Conversation.Entry.RoyalWonk}}
 }
 Speech_C_4 {
  id=#2
  class=Speech_C
  outputs [#3, #4, #5, #6]
  properties {
   nodeName=""
   generalConfig {
    defaultMessage="Stay Wonkee!"
    contextualMessages [
     {
      message="Welcome to the wonkee-est place on the Island! WOO-HOO-HOO!"
      contextRequirements [{requirement {class=FortControllerRequirement_IsFirstEverConversationWithNPC}}]
      priority {curve {rowName=Tandem.Speech.Priority.FirstEverInteraction}}
     }
     {
      message="Stay Wonkee!"
      contextRequirements [{requirement {class=FortControllerRequirement_CheckNPCMatchConversationCount, properties {nPCUniqueIdTag {tagName=AISpawnerData.Type.Tandem.RoyalWonk}}}}]
      priority {curve {curveTable=null, rowName=None}}
     }
     {
      message="At your Wonkee-est service!"
      contextRequirements [{requirement {class=FortControllerRequirement_CheckNPCMatchConversationCount, properties {nPCUniqueIdTag {tagName=AISpawnerData.Type.Tandem.RoyalWonk}}}}]
      priority {curve {curveTable=null, rowName=None}}
     }
     {
      message="W-O-N-K-E-E!!"
      contextRequirements [{requirement {class=FortControllerRequirement_CheckNPCMatchConversationCount, properties {nPCUniqueIdTag {tagName=AISpawnerData.Type.Tandem.RoyalWonk}}}}]
      priority {curve {curveTable=null, rowName=None}}
     }
     {
      message="The most powerful relic at Wonkeeland is a smile!"
      contextRequirements [{requirement {
         class=FortControllerRequirement_HasCID
         properties {allowedCIds [
           /BRCosmetics/Athena/Items/Cosmetics/Characters/CID_288_Athena_Commando_M_IceKing
           /BRCosmetics/Athena/Items/Cosmetics/Characters/Character_BriskReign
          ]}
        }}]
      priority {curve {rowName=Tandem.Speech.Priority.OutfitReaction}}
     }
     {
      message=Seriously? We JUST cleaned up what you did to the lazy river.
      contextRequirements [{requirement {
         class=FortControllerRequirement_HasCID
         properties {allowedCIds [/BRCosmetics/Athena/Items/Cosmetics/Characters/Character_SteedShimmer_Shire]}
        }}]
      priority {curve {rowName=Tandem.Speech.Priority.OutfitReaction}}
     }
     {
      message=Please... I have a wonkee family at home...
      contextRequirements [{requirement {
         class=FortControllerRequirement_HasCID
         properties {allowedCIds [
           /BRCosmetics/Athena/Items/Cosmetics/Characters/Character_SpacePlunge
           /BRCosmetics/Athena/Items/Cosmetics/Characters/CID_088_Athena_Commando_M_SpaceBlack
          ]}
        }}]
      priority {curve {rowName=Tandem.Speech.Priority.OutfitReaction}}
     }
    ]
   }
  }
 }
 ConversationLinkNode_3 {id=#6, class=ConversationLinkNode, properties {remoteEntryTag {tagName=Conversation.Entry.RoyalWonk.Dialogue}}}
 ConversationLinkNode_2 {id=#5, class=ConversationLinkNode, properties {remoteEntryTag {tagName=Conversation.Entry.RandomServices}}}
 ConversationLinkNode_0 {id=#4, class=ConversationLinkNode, properties {remoteEntryTag {tagName=Conversation.Entry.BasicServices}}}
 ConversationLinkNode_1 {id=#3, class=ConversationLinkNode, properties {remoteEntryTag {tagName=Conversation.Entry.NPCLibrary.Quests}}}
]
